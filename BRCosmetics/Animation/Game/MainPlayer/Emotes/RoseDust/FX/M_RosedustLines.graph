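# Unlit, two-sided translucent Niagara sprite and mesh particle material for line streaks in the RoseDust emote, with particle color driving emissive.
# Dynamic parameters pan the line texture along U and tile both axes by Xtile, leaving Ytile unused.
# Only red channels of texture and ChannelPicker feed the dot product.
# A BandCut-blended sphere mask trims U ends before a narrow SmoothStep hardens opacity, scaled by particle alpha.
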Material /BRCosmetics/Animation/Game/MainPlayer/Emotes/RoseDust/FX/M_RosedustLines
graph {
 properties {
  bUsedWithNiagaraMeshParticles=true
  bUsedWithNiagaraSprites=true
  blendMode=BLEND_Translucent
  pixelDepthOffsetMode=PDOM_Legacy
  shadingModel=MSM_Unlit
  twoSided=true
  usedShadingModels=Unlit
 }
 nodes [
  n1 = ParticleColor()
  MaterialFunctionCall(Emissive=@n1.RGB) {materialFunction=&MF_FX_EmissiveAdjustments}
  n2 = Constant() {r=0.25}
  n3 = Add(A=@n2) {constB=0.02}
  n4 = TextureCoordinate()
  n5 = DynamicParameter() {paramNames [Offset,Xtile,Ytile,Param4]}
  n6 = AppendVector(A=@n5.Xtile, B=@n5.Xtile)
  n7 = TextureSample(UVs=Panner(Coordinate=Multiply(A=@n4, B=@n6), Time=@n5.Offset) {speedX=1}) {samplerType=SAMPLERTYPE_LinearColor, texture=&/T_RosedustLines_M}
  n8 = VectorParameter() {parameterName=ChannelPicker, defaultValue=<rgba 1 0 0 1>}
  n9 = TextureCoordinate() {vTiling=0}
  n10 = Constant2Vector() {r=0.5}
  n11 = Constant() {r=0.5}
  n12 = Constant()
  n13 = Multiply(
   A=DotProduct(
    A=MaterialFunctionCall(X=@n7.R, Y=@n7.R, Z=@n7.R) {materialFunction=&MakeFloat3}.Result
    B=MaterialFunctionCall(X=@n8.R, Y=@n8.R, Z=@n8.R) {materialFunction=&MakeFloat3}.Result
   )
   B=LinearInterpolate(B=SphereMask(A=@n9, B=@n10, Radius=@n11, Hardness=@n12), Alpha=ScalarParameter() {defaultValue=1, parameterName=BandCut}) {constA=1, constAlpha=1}
  )
  Saturate(None=Multiply(A=SmoothStep(Min=@n2, Max=@n3, Value=@n13), B=@n1.A))
  ScalarParameter() {parameterName=Offset}
 ]
}
