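# Translucent unlit Niagara ribbon material for the RoseDust emote trail.
# A panning Voronoi texture erodes a U gradient, combined with panning Perlin noise streaks that taper along U and fade in at the start.
# The result is scaled by particle color and alpha, with the color passing through MF_FX_EmissiveAdjustments.
# A reroute note records a change from masked to translucent blending.
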Material /BRCosmetics/Animation/Game/MainPlayer/Emotes/RoseDust/FX/M_Rosedust_Emote_Trail
graph {
 properties {bUsedWithNiagaraRibbons=true, blendMode=BLEND_Translucent, pixelDepthOffsetMode=PDOM_Legacy, shadingModel=MSM_Unlit, usedShadingModels=Unlit}
 nodes [
  n1 = TextureCoordinate()
  n2 = TextureSample(UVs=Panner(Coordinate=TextureCoordinate() {coordinateIndex=1, uTiling=0.5, vTiling=0.5}) {speedX=-1}) {texture=/BRCosmetics/Athena/Effects/Particles/Skydiving/NS_Contrail_PinkSpike/Textures/T_PinkSpike_Voronoi2D_Height}
  n3 = OneMinus(None=AppendVector(A=@n2.R, B=@n2.R))
  n4 = ComponentMask(None=OneMinus(None=Add(A=@n1, B=@n3))) {r=true}
  n5 = OneMinus(None=@n1)
  n6 = Panner(Coordinate=TextureCoordinate() {coordinateIndex=1, uTiling=0}) {speedX=0.5, speedY=0.5}
  n7 = Subtract(A=Multiply(A=ComponentMask(None=@n5) {r=true}, B=TextureSample(UVs=@n6) {texture=/Game/Global/Textures/Default/Noise/3xPerlin_M}.G)) {constB=0.05}
  n8 = SmoothStep(Value=ComponentMask(None=@n1) {r=true}) {constMax=0.1}
  n9 = Saturate(None=Multiply(A=Power(Base=@n7), B=@n8))
  n10 = Saturate(None=Add(A=@n4, B=Multiply(A=@n9) {constB=2}))
  n11 = ParticleColor()
  MaterialFunctionCall(Emissive=Multiply(A=@n10, B=@n11.RGB)) {materialFunction=&MF_FX_EmissiveAdjustments}
  n12 = Multiply(A=@n10, B=@n11.A)
  Reroute(None=@n12) {desc="I used to be masked, now I am translucent"}
  Saturate(None=Divide(A=Subtract(A=@n12) {constB=0.266}) {constB=0.066})
  Subtract(A=Multiply(A=@n4, B=@n8)) {constB=0.3}
 ]
}
