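# Niagara system for the RoseDust emote that draws short fading ribbon blur trails behind the player's hands and feet.
# Two emitters each spawn and update two anim trails from bone or socket pairs on User.SkeletalMesh, with width and alpha tapering over particle life.
# Scale values drive trail width and lifetime: feet read User.FeetTrailScale, which defaults to zero, while hands read System.HandTrailScale.
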NiagaraSystem /BRCosmetics/Animation/Game/MainPlayer/Emotes/RoseDust/FX/NS_Rosedust_Emote_Blurlines
dependencies {
 usedDataInterfaces [&NiagaraDataInterfaceCurve, &NiagaraDataInterfaceArrayFloat3]
 usedDynamicInputs []
 usedModules [
  &SystemState
  &self:SystemSpawnScript.SystemScriptSource.SystemScriptGraph.NiagaraNodeAssignment_0.SetVariables_#1
 ]
}
emitters {
 Trail_Feet {
  emitterUpdateScript [
   &EmitterState()
   &FN_SpawnAnimTrail(AnimTrailWidthMode=::FromFirst, Number of Anim Trails=2, Spawn Interval=2, Spawn Mode:ENiagaraSkeletalSpawning=::NewEnumerator0)
  ]
  particleSpawnScript [
   &FN_InitializeAnimTrailBoneIndecies(First Bone or Socket Index=4, Second Bone or Socket Index=5)
   &FN_InitializeAnimTrailBoneIndecies(AnimTrail Selection=1, First Bone or Socket Index=6, Second Bone or Socket Index=7) {module=FN_InitializeAnimTrailBoneIndecies001}
   &FN_InitializeAnimTrail(
    AnimTrailWidth=&Lerp_Float(A=0, B=0.5, Alpha=$User.FeetTrailScale)
    Color=<rgba 0 0 0 1>
    Lifetime=&Lerp_Float(
     A=0
     B=0.3
     Alpha=&FloatFromCurve(
      FloatCurve=DataInterface() {propertyValues {Curve {keys [
          {arriveTangent=-1.5058, interpMode=RCIM_Cubic, leaveTangent=-1.5058, tangentMode=RCTM_User, time=0, value=1}
          {interpMode=RCIM_Cubic, tangentMode=RCTM_User, time=1, value=0}
         ]}}}
      Scale Curve=$User.FeetTrailScale
     )
    )
    Position Offset=&SelectVectorFromArray(Direct Array Index=$Particles.SpawnGroup, Vector Selection Array=DataInterface() {propertyValues {FloatData [<2 0 0>, <2 0 0>]}})
   )
  ]
  particleUpdateScript [
   &SolveForcesAndVelocity(Clamp Velocity=false, Limit Acceleration=false)
   &ParticleState()
   &FN_UpdateAnimTrail(SkeletalMesh=$User.SkeletalMesh)
   &ScaleRibbonWidth(Ribbon Width Scale=&FloatFromCurve(FloatCurve=DataInterface() {propertyValues {Curve {keys [
         {arriveTangent=6.4514, interpMode=RCIM_Cubic, leaveTangent=6.4514, tangentMode=RCTM_User, time=0, value=0.5}
         {arriveTangent=-0.7143, interpMode=RCIM_Cubic, leaveTangent=-0.7143, time=0.1, value=1}
         {interpMode=RCIM_Cubic, tangentMode=RCTM_User, time=0.7, value=0}
        ]}}}))
   &ScaleColor(
    Scale Alpha=&FloatFromCurve(FloatCurve=DataInterface() {propertyValues {Curve {keys [
         {arriveTangent=-2.4147, interpMode=RCIM_Cubic, leaveTangent=-2.4147, tangentMode=RCTM_User, time=0, value=1}
         {interpMode=RCIM_Cubic, tangentMode=RCTM_User, time=0.7, value=0}
        ]}}})
    ScaleA=true
    ScaleRGB=true
   )
  ]
  rendererClasses [&NiagaraRibbonRendererProperties]
  renderers [{rendererClass=&NiagaraRibbonRendererProperties}]
 }
 Trail_Hands {
  emitterUpdateScript [
   &EmitterState()
   &FN_SpawnAnimTrail(AnimTrailWidthMode=::FromFirst, Number of Anim Trails=2, Spawn Interval=2, Spawn Mode:ENiagaraSkeletalSpawning=::NewEnumerator0)
  ]
  particleSpawnScript [
   &FN_InitializeAnimTrailBoneIndecies()
   &FN_InitializeAnimTrailBoneIndecies(AnimTrail Selection=1, First Bone or Socket Index=2, Second Bone or Socket Index=3) {module=FN_InitializeAnimTrailBoneIndecies001}
   &FN_InitializeAnimTrail(
    AnimTrailWidth=&Lerp_Float(A=0, B=0.5, Alpha=$System.HandTrailScale)
    Lifetime=&Lerp_Float(
     A=0
     B=0.4
     Alpha=&FloatFromCurve(
      FloatCurve=DataInterface() {propertyValues {Curve {keys [
          {arriveTangent=-1.836, interpMode=RCIM_Cubic, leaveTangent=-1.836, tangentMode=RCTM_User, time=0, value=1}
          {interpMode=RCIM_Cubic, tangentMode=RCTM_User, time=1, value=0}
         ]}}}
      Scale Curve=$System.HandTrailScale
     )
    )
    Position Offset=&SelectVectorFromArray(Direct Array Index=$Particles.SpawnGroup, Vector Selection Array=DataInterface() {propertyValues {FloatData [<0 0 0.2>, <0 0 0.2>]}})
   )
  ]
  particleUpdateScript [
   &SolveForcesAndVelocity(Clamp Velocity=false, Limit Acceleration=false)
   &ParticleState()
   &FN_UpdateAnimTrail(SkeletalMesh=$User.SkeletalMesh)
   &ScaleRibbonWidth(Ribbon Width Scale=&FloatFromCurve(FloatCurve=DataInterface() {propertyValues {Curve {keys [
         {arriveTangent=6.4514, interpMode=RCIM_Cubic, leaveTangent=6.4514, tangentMode=RCTM_User, time=0, value=0.5}
         {arriveTangent=-0.7143, interpMode=RCIM_Cubic, leaveTangent=-0.7143, time=0.1, value=1}
         {interpMode=RCIM_Cubic, tangentMode=RCTM_User, time=0.7, value=0}
        ]}}}))
   &ScaleColor(
    Scale Alpha=&FloatFromCurve(FloatCurve=DataInterface() {propertyValues {Curve {keys [
         {arriveTangent=-2.4147, interpMode=RCIM_Cubic, leaveTangent=-2.4147, tangentMode=RCTM_User, time=0, value=1}
         {interpMode=RCIM_Cubic, tangentMode=RCTM_User, time=0.7, value=0}
        ]}}})
    ScaleA=true
    ScaleRGB=true
   )
  ]
  rendererClasses [&NiagaraRibbonRendererProperties]
  renderers [{rendererClass=&NiagaraRibbonRendererProperties}]
 }
}
summary {
 systemName=NS_Rosedust_Emote_Blurlines
 emitters [
  {emitterName=Trail_Hands, rendererClasses [&NiagaraRibbonRendererProperties]}
  {emitterName=Trail_Feet, rendererClasses [&NiagaraRibbonRendererProperties]}
 ]
 userVariables [
  User.FeetTrailScale {type=NiagaraFloat, defaultValue=0, description=""}
  User.HandTrailScale {type=NiagaraFloat, defaultValue=1, description=""}
  User.RibbonIDOffset {type=NiagaraInt32}
  User.SkeletalMesh {
   type=NiagaraDataInterfaceSkeletalMesh
   defaultValue {
    struct=NiagaraExt_VariableValue_DataInterface
    value {dataInterface=&self:NiagaraDataInterfaceSkeletalMesh_0}
   }
  }
 ]
}
